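# GreaseClip SFX subsequence is a Level Sequence that, by its name and folder, holds the sound effects layer of the GreaseClipDunmurry prototype sequence.
# It has no object bindings, only one root audio track with a single section over frames 0 to 318 at a 30 fps display rate.
# The section leaves volume, pitch and attach actor at plain defaults, and no sound asset is shown.
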LevelSequence /GreaseClipDunmurry/ProtoSeq/Subsequences/GreaseClip_SFX_Subsequence
bindings []
display_rate {denominator=1, numerator=30}
tick_resolution {denominator=1, numerator=24000}
tracks [{
  type=MovieSceneAudioTrack
  display_name=Audio
  sections [{
    type=MovieSceneAudioSection
    end_frame=318
    start_frame=0
    channels [
     Volume {type=MovieSceneScriptingFloatChannel, default=1}
     Pitch {type=MovieSceneScriptingFloatChannel, default=1}
     AttachActor {type=MovieSceneScriptingActorReferenceChannel, default="<Struct 'MovieSceneObjectBindingID' (0x49ea84f7bc0) {}>"}
    ]
   }]
 }]
